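# Niagara one-shot effect for the Brawny Smasher NPC's swap, combining a blue-cyan burst with effects across the skeleton.
# Flare, Light, mesh Fragments and GPU Embers share one offset socket sampled from User.SkeletalMesh.
# Shield_Decals pins 19 fading sprites to separate bones or sockets, using per-index offsets from an array.
# Ring places delayed mesh particles via FN_BoneFlow on User.BoneFlow, spread by spawn group, with waveform-driven color and tangent-based rotation.
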NiagaraSystem /FNGameplayCues/Athena/AI/NPCs/Tandem/Base/Items/FX/NS_BrawnySmasher_SwapFX
dependencies {
 usedDataInterfaces [&NiagaraDataInterfaceCurve, &NiagaraDataInterfaceArrayFloat3]
 usedDynamicInputs []
 usedModules [
  &SystemState
  &self:SystemSpawnScript.SystemScriptSource.SystemScriptGraph.NiagaraNodeAssignment_0.SetVariables_#1
 ]
}
emitters {
 Embers {
  simTarget=GPUComputeSim
  emitterUpdateScript [
   &EmitterState()
   &SpawnBurst_Instantaneous(Spawn Count=&RandomRangeInt(Minimum=5, Maximum=10), Spawn Time=0.05, Use Spawn Probability=false)
  ]
  particleSpawnScript [
   &InitializeParticle(
    Color=<rgba 1.276 3.7209 4 1>
    Lifetime Mode=::NewEnumerator1
    Sprite Rotation Mode=::NewEnumerator2
    Sprite Size Mode=::NewEnumerator1
    Uniform Sprite Size Max=32
    Uniform Sprite Size Min=7
    UsePositionOffset=false
    Write Material Random=true
   )
   &SkeletalMeshLocation(
    Bone / Socket Index=15
    Bone Sampling Mode=::NewEnumerator4
    MaskSpawnGroup=false
    Sampled Position Offset=<5 0 0>
    Skeletal Mesh=$User.SkeletalMesh
   )
   &AddVelocityFromPoint(
    Default Position:Vector3f=/Niagara/DynamicInputs/Helpers/SimulationPosition()
    Module.Use Falloff Distance=false
    Velocity Strength=&RandomRangeFloat(Minimum=500, Maximum=1000)
   )
  ]
  particleUpdateScript [
   &GravityForce(Gravity=&RandomRangeVector(Minimum=<0 0 -50>, Maximum=<0 0 -150>))
   &Drag(Drag=&RandomRangeFloat(Minimum=1, Maximum=5), Use Linear Drag=true, Use Rotational Drag=false)
   &CurlNoiseForce(Noise Frequency=5, Noise Strength=250, Pan Noise=true, Pan Noise Field=<5 5 5>)
   &SolveForcesAndVelocity(Clamp Velocity=false, Limit Acceleration=false)
   &ParticleState()
   &DynamicMaterialParameters(
    Index 0 Param 2=2
    Module.Param0WriteEnabled=true
    Module.Param1WriteEnabled=true
    Module.Param2WriteEnabled=true
    Module.Param3WriteEnabled=true
   )
   &SpriteRotationRate(Rotation Rate=&RandomRangeFloat(Minimum=-90, Maximum=90))
   &ScaleSpriteSize(Scale Factor=&Vector2DFromFloat(Value=&FloatFromCurve(FloatCurve=DataInterface() {propertyValues {Curve {keys [
          {arriveTangent=2.0259, interpMode=RCIM_Cubic, leaveTangent=2.0259, tangentMode=RCTM_User, time=0, value=0}
          {arriveTangentWeight=0.1574, interpMode=RCIM_Cubic, leaveTangentWeight=0.1427, tangentMode=RCTM_User, time=0.1, value=0.45}
          <at 0.2013 0.4252>
          <at 0.2081 0.6346>
          <at 0.2204 0.2868>
          <at 0.2311 0.5811>
          <at 0.3936 0.7212>
          <at 0.401 0.895>
          <at 0.4103 0.7212>
          <at 0.6261 0.8137>
          <at 0.6334 0.9987>
          <at 0.6474 0.7156>
          {arriveTangent=-6.7094, interpMode=RCIM_Cubic, leaveTangent=-6.7094, tangentMode=RCTM_User, time=0.6514, value=0.9398}
          <at 1 0 RCIM_Cubic>
         ]}}})))
  ]
  rendererClasses [&NiagaraSpriteRendererProperties]
  renderers [{rendererClass=&NiagaraSpriteRendererProperties}]
 }
 Flare {
  emitterUpdateScript [
   &EmitterState()
   &SpawnBurst_Instantaneous(Spawn Count=1, Use Spawn Probability=false)
  ]
  particleSpawnScript [
   &InitializeParticle(
    Color=<rgba 0.3636 1.4091 2 1>
    Lifetime=0.55
    Mesh Scale Mode=::NewEnumerator3
    Mesh Uniform Scale=0.4
    Sprite Rotation Mode=::NewEnumerator2
    Sprite Size Mode=::NewEnumerator3
    Uniform Sprite Size=320
    UsePositionOffset=false
    Write Material Random=true
   )
   &CameraOffset(Camera Offset Amount=55)
   /SetVariables() {isSetParameters=true}
  ]
  particleUpdateScript [
   &ParticleState()
   &DynamicMaterialParameters(
    Index 0 Param 1=&RandomRangeFloat(Minimum=0.4, Maximum=0.6)
    Index 0 Param 2=1
    Index 0 Param 3=1
    Index 2 Param 1=&FloatFromCurve(FloatCurve=DataInterface() {propertyValues {Curve {keys [
         {arriveTangent=-4.8372, interpMode=RCIM_Cubic, leaveTangent=-4.8372, tangentMode=RCTM_User, time=0, value=1.5}
         {arriveTangent=-0.4216, interpMode=RCIM_Cubic, leaveTangent=-0.4216, tangentMode=RCTM_User, time=0.1, value=1.06}
         <at 1 0.952 RCIM_Cubic>
        ]}}})
    Module.Param0WriteEnabled=true
    Module.Param0WriteEnabled2=true
    Module.Param1WriteEnabled=true
    Module.Param1WriteEnabled2=true
    Module.Param2WriteEnabled=true
    Module.Param2WriteEnabled2=true
    Module.Param3WriteEnabled=false
    Module.Param3WriteEnabled2=true
    Write Parameter Index 2=true
   )
   &SkeletalMeshLocation(
    Bone / Socket Index=15
    Bone Sampling Mode=::NewEnumerator4
    MaskSpawnGroup=false
    Sampled Position Offset=<5 0 0>
    Skeletal Mesh=$User.SkeletalMesh
   )
   &ScaleSpriteSize(Scale Factor=&Vector2DFromFloat(Value=&FloatFromCurve(FloatCurve=DataInterface() {propertyValues {Curve {keys [
          {arriveTangent=260.8905, interpMode=RCIM_Cubic, leaveTangent=260.8905, tangentMode=RCTM_User, time=0, value=0}
          <at 0.01 1 RCIM_Cubic>
          {arriveTangent=-3.1687, interpMode=RCIM_Cubic, leaveTangent=-3.1687, tangentMode=RCTM_User, time=1, value=0}
         ]}}})))
   &ScaleMeshSize(Scale Factor=&VectorFromFloat(Value=&FloatFromCurve(FloatCurve=DataInterface() {propertyValues {Curve {keys [{arriveTangent=4.4533, interpMode=RCIM_Cubic, leaveTangent=4.4533, tangentMode=RCTM_User, time=0, value=0}, <at 1 1.5 RCIM_Cubic>]}}})))
   /SetVariables(Particles.Mesh Color:LinearColor=&MakeLinearColorFromVectorAndFloat(
     "Vector (RGB)"=&MakeVectorFromLinearColorRGB(Color=<rgba 0 0.6095 1 1>)
     "Float (Alpha)"=&FloatFromCurve(
      FloatCurve=DataInterface() {propertyValues {Curve {keys [
          {arriveTangent=259.0315, interpMode=RCIM_Cubic, leaveTangent=259.0315, tangentMode=RCTM_User, time=0, value=0}
          {arriveTangent=1.75, interpMode=RCIM_Cubic, leaveTangent=1.75, time=0.01, value=1}
          {arriveTangent=-2.629, interpMode=RCIM_Cubic, leaveTangent=-2.629, tangentMode=RCTM_User, time=0.4, value=0.7}
          <at 1 0 RCIM_Cubic>
         ]}}}
      Scale Curve=0.5
     )
    )) {isSetParameters=true}
   &ScaleColor(
    Scale Alpha=&FloatFromCurve(FloatCurve=DataInterface() {propertyValues {Curve {keys [
         {interpMode=RCIM_Cubic, tangentMode=RCTM_User, time=0, value=0}
         {interpMode=RCIM_Cubic, tangentMode=RCTM_User, time=0.2, value=1}
         {arriveTangentWeight=0.3202, interpMode=RCIM_Cubic, leaveTangentWeight=0.1067, tangentMode=RCTM_User, time=0.8, value=1}
         {tangentMode=RCTM_User, time=1, value=0}
        ]}}})
    ScaleA=true
    ScaleRGB=true
   )
  ]
  rendererClasses [&NiagaraSpriteRendererProperties, &NiagaraMeshRendererProperties]
  renderers [
   {rendererClass=&NiagaraSpriteRendererProperties}
   {rendererClass=&NiagaraMeshRendererProperties, rendererIndex=1}
  ]
 }
 Fragments {
  emitterUpdateScript [
   &EmitterState()
   &SpawnBurst_Instantaneous(Spawn Count=&RandomRangeInt(Minimum=10, Maximum=20), Spawn Time=0.05, Use Spawn Probability=false)
  ]
  particleSpawnScript [
   &InitializeParticle(
    Color=<rgba 1.7058 6.5057 10 2>
    Lifetime Min=0.5
    Lifetime Mode=::NewEnumerator1
    Mesh Scale Max=<1 2 3>
    Mesh Scale Min=<0.5 0.75 2>
    Mesh Scale Mode=::NewEnumerator2
    UsePositionOffset=false
    Write Material Random=true
   )
   &SkeletalMeshLocation(
    Bone / Socket Index=15
    Bone Sampling Mode=::NewEnumerator4
    MaskSpawnGroup=false
    Sampled Position Offset=<5 0 0>
    Skeletal Mesh=$User.SkeletalMesh
   )
   &SphereLocation(
    MaskSpawnGroup=false
    Override Seed=false
    Sphere Radius=&RandomRangeFloat(Minimum=2, Maximum=5)
    Surface Only=false
    UseNonUniformScale=false
    UseOffset=false
    UseSphereOrientation=false
   )
   &AddVelocityFromPoint(
    Default Position:Vector3f=&SimulationPosition()
    Module.Use Falloff Distance=false
    Origin Offset Coordinate Space=::Simulation
    Velocity Strength=&RandomRangeFloat(Minimum=500, Maximum=750)
   )
   &InitialMeshOrientation(
    Rotation=&RandomRangeVector(Minimum=<0 0 0>, Maximum=<1 1 1>)
    Use Orientation Vector=true
    Use Rotation Vector=true
   )
  ]
  particleUpdateScript [
   &ParticleState()
   &GravityForce(Gravity=&RandomRangeVector(Minimum=<0 0 -50>, Maximum=<0 0 -200>))
   &Drag(Drag=&RandomRangeFloat(Minimum=1, Maximum=5), Use Linear Drag=true, Use Rotational Drag=false)
   &SolveForcesAndVelocity(Clamp Velocity=false, Limit Acceleration=false)
   &UpdateMeshOrientation(
    Rotation Rate=&FloatFromCurve(
     FloatCurve=DataInterface() {propertyValues {Curve {keys [{arriveTangent=-2.6547, interpMode=RCIM_Cubic, leaveTangent=-2.6547, tangentMode=RCTM_User, time=0, value=1}, <at 1 0.1 RCIM_Cubic>]}}}
     Scale Curve=10
    )
    Rotation Vector=&RandomRangeVector(Minimum=<0 0 0>, Maximum=<1 1 1>)
   )
   &ScaleMeshSize(Scale Factor=&VectorFromFloat(Value=&FloatFromCurve(FloatCurve=DataInterface() {propertyValues {Curve {keys [
          {arriveTangent=22.3133, interpMode=RCIM_Cubic, leaveTangent=22.3133, tangentMode=RCTM_User, time=0, value=0}
          {arriveTangentWeight=0.0374, interpMode=RCIM_Cubic, leaveTangentWeight=0.3895, tangentMode=RCTM_User, time=0.07, value=1}
          {arriveTangentWeight=0.3573, interpMode=RCIM_Cubic, leaveTangentWeight=0.0979, tangentMode=RCTM_User, time=0.8, value=1}
          <at 1 0 RCIM_Cubic>
         ]}}})))
  ]
  rendererClasses [&NiagaraMeshRendererProperties]
  renderers [{rendererClass=&NiagaraMeshRendererProperties, rendererIndex=0}]
 }
 Light {
  emitterUpdateScript [
   &EmitterState()
   &SpawnBurst_Instantaneous(Spawn Count=1, Use Spawn Probability=false)
  ]
  particleSpawnScript [
   &InitializeParticle(Color Mode=::NewEnumerator0, UsePositionOffset=false, Write Material Random=true)
   &SkeletalMeshLocation(
    Bone / Socket Index=15
    Bone Sampling Mode=::NewEnumerator4
    MaskSpawnGroup=false
    Sampled Position Offset=<5 0 0>
    Skeletal Mesh=$User.SkeletalMesh
   )
  ]
  particleUpdateScript [
   &ParticleState()
   &FN_Light_Attributes(
    Color=<rgba 0.319 0.9302 1 1>
    Create Color Attribute=true
    Create Exponent Attribute=true
    Create Radius Attribute=true
    Exponent=3
    Radius=200
    Scale Color=&FloatFromCurve(
     FloatCurve=DataInterface() {propertyValues {Curve {keys [
         {arriveTangent=4.0385, interpMode=RCIM_Cubic, leaveTangent=4.0385, tangentMode=RCTM_User, time=0, value=0}
         {arriveTangent=-1.6622, interpMode=RCIM_Cubic, leaveTangent=-1.6622, tangentMode=RCTM_User, time=0.4, value=1}
         <at 1 0 RCIM_Cubic>
        ]}}}
     Scale Curve=50
    )
   )
  ]
  rendererClasses [&NiagaraLightRendererProperties]
  renderers [{rendererClass=&NiagaraLightRendererProperties, rendererIndex=0}]
 }
 Ring {
  emitterSpawnScript [/SetVariables(Emitter.UseParticleColorInput=1) {isSetParameters=true}]
  emitterUpdateScript [
   &EmitterState(Life Cycle Mode=::NewEnumerator1, Loop Behavior=::NewEnumerator1, Loop Delay=0.21, Loop Duration=1, UseLoopDelay=true)
   &SpawnBurst_Instantaneous(Spawn Count=20, Use Spawn Probability=false)
  ]
  particleSpawnScript [
   &FN_SetSpawnGroup(Inclusivity=true, Normalize Spawn Group=true, Pattern=::NewEnumerator0, Set Ribbon ID and Ribbon Link Order=true) {module=FN_SetSpawnGroup001}
   &InitializeParticle(
    Color=<rgba 0 1 0.0615 10>
    Lifetime Max=1
    Lifetime Min=0.65
    Lifetime Mode=::NewEnumerator1
    Mesh Scale Mode=::NewEnumerator3
    Mesh Uniform Scale=0.05
    UsePositionOffset=false
    Write Material Random=true
   )
   &CameraOffset(Camera Offset Amount=8)
  ]
  particleUpdateScript [
   &ParticleState()
   /Game/Effects/Niagara/Modules/BoneFlow/FN_BoneFlow_v4/FN_BoneFlow(
    Angle:NiagaraFloat=&Multiply_Float(A=$Particles.Normalized Spawn Group, B=1)
    BoneID:NiagaraInt32=&ModuloInt(A=$Particles.SpawnGroup, B=5)
    Height:NiagaraFloat=&Multiply_Float(A=$Particles.Normalized Spawn Group, B=0.9)
    Interpolate Mesh Sampling=false
    Radius:NiagaraFloat=&Multiply_Float(A=$Particles.Normalized Spawn Group, B=0.3)
    SkeletalMesh:NiagaraDataInterfaceSkeletalMesh=$User.BoneFlow
    Write Position Attribute=true
   ) {module=FN_BoneFlow001}
   /SetVariables(Particles.SizeWave:NiagaraFloat=&Waveform(
     Global Input Position=&Multiply_Float(A=$Engine.Time, B=3)
     Global Phase=$Particles.NormalizedAge
     "Waveform [1]"=::NewEnumerator0
     "[1] Phase"=$Particles.NormalizedAge
     "[1] Frequency"=0.3
     "[1] Amplitude Min/Max"=<1 0>
     "[1] Scale"=1.5
     "[1] Bias"=1
     "[2] Phase"=$Emitter.Age
     "[3] Phase"=$Emitter.Age
    )) {isSetParameters=true}
   /SetVariables(Particles.Desaturation=0.6) {isSetParameters=true}
   /SetVariables(Particles.WaveColor:LinearColor=&MakeLinearColor(
     R:NiagaraFloat=&Waveform(
      Global Input Position=&Multiply_Float(A=$Engine.Time, B=1.5)
      Global Phase=$Particles.NormalizedAge
      "Waveform [1]"=::NewEnumerator0
      "[1] Phase"=$Particles.NormalizedAge
      "[1] Frequency"=1.5
      "[1] Amplitude Min/Max"=<1 0>
      "[1] Bias"=$Particles.Desaturation
      "[2] Phase"=$Emitter.Age
      "[3] Phase"=$Emitter.Age
     )
     G:NiagaraFloat=&Waveform(
      Global Input Position=&Multiply_Float(A=$Engine.Time, B=1)
      Global Phase=$Particles.NormalizedAge
      "Waveform [1]"=::NewEnumerator1
      "[1] Phase"=$Particles.NormalizedAge
      "[1] Frequency"=1
      "[1] Amplitude Min/Max"=<1 0>
      "[1] Scale"=2
      "[1] Bias"=$Particles.Desaturation
      "[2] Phase"=$Emitter.Age
      "[3] Phase"=$Emitter.Age
     )
     B:NiagaraFloat=&Waveform(
      Global Input Position=&Multiply_Float(A=$Engine.Time, B=1.5)
      Global Phase=$Particles.NormalizedAge
      "Waveform [1]"=::NewEnumerator2
      "[1] Phase Scale"=2.5
      "[1] Phase"=$Particles.NormalizedAge
      "[1] Frequency"=2
      "[1] Amplitude Min/Max"=<1 0>
      "[1] Bias"=$Particles.Desaturation
      "[2] Phase"=$Emitter.Age
      "[3] Phase"=$Emitter.Age
     )
     A:NiagaraFloat=1
    )) {isSetParameters=true}
   /SetVariables(Particles.Color=&MultiplyLinearColor(
     Linear Color=$Particles.WaveColor
     Scale Factor=&MakeVector4_FromLinearColor(LinearColor=&MakeLinearColorFromVectorAndFloat("Vector (RGB)"=<0.2 0.5 1>, "Float (Alpha)"=5))
    )) {isSetParameters=true}
   /SetVariables(Particles.BoneFlowTangent:Vector3f=$Output.FN_BoneFlow001.SplineTangent) {isSetParameters=true}
   /SetVariables(Particles.RingRotation:Quat4f=&MakeQuatFromAxisAngle(Axis=$Particles.BoneFlowTangent, AngleInDegrees=120, NormalizeAxis=true)) {isSetParameters=true}
   &ScaleMeshSize(Scale Factor=&Multiply_VectorByFloat(
     Vector=<1 1 1>
     Float=&FloatFromCurve(FloatCurve=DataInterface() {propertyValues {Curve {keys [{arriveTangent=11.048, interpMode=RCIM_Cubic, time=0, value=1}, {interpMode=RCIM_Cubic, leaveTangent=8.0098, time=1, value=4}]}}})
    )) {module=ScaleMeshSize001}
   &DynamicMaterialParameters(
    Index 0 Param 1=&FloatFromCurve(FloatCurve=DataInterface() {propertyValues {Curve {keys [
         {arriveTangent=0.1068, interpMode=RCIM_Cubic, leaveTangent=0.1068, tangentMode=RCTM_User, time=0.1515, value=0.1}
         {arriveTangent=-0.0405, interpMode=RCIM_Cubic, leaveTangent=-0.0405, tangentMode=RCTM_User, time=1, value=1}
        ]}}})
    Index 0 Param 2=&FloatFromCurve(FloatCurve=DataInterface() {propertyValues {Curve {keys [<at 0 1 RCIM_Cubic>, {arriveTangent=-3.9609, interpMode=RCIM_Cubic, leaveTangent=-3.9609, tangentMode=RCTM_User, time=0.8, value=-0.25}]}}})
    Module.Param0WriteEnabled=true
    Module.Param1WriteEnabled=true
    Module.Param2WriteEnabled=true
    Module.Param3WriteEnabled=true
   )
  ]
  rendererClasses [&NiagaraMeshRendererProperties]
  renderers [{rendererClass=&NiagaraMeshRendererProperties, rendererIndex=0}]
 }
 Shield_Decals {
  emitterUpdateScript [
   &EmitterState()
   &SpawnBurst_Instantaneous(Spawn Count=19, Use Spawn Probability=false)
  ]
  particleSpawnScript [
   &FN_SetSpawnGroup(Number of Spawn Groups=19)
   &InitializeParticle(
    Color=<rgba 0.0869 0.1596 1 1>
    Lifetime=1.3
    Sprite Size Mode=::NewEnumerator3
    Uniform Sprite Size=75
    UsePositionOffset=false
    Write Material Random=true
   )
   &CameraOffset(Camera Offset Amount=25)
  ]
  particleUpdateScript [
   &ParticleState()
   &SkeletalMeshLocation(
    Bone / Socket Index=$Particles.SpawnGroup
    Bone Sampling Mode=::NewEnumerator4
    MaskSpawnGroup=false
    Sampled Position Offset=&SelectVectorFromArray(
     Direct Array Index=$Particles.SpawnGroup
     Vector Selection Array=DataInterface() {propertyValues {FloatData [
        <0 0 0>
        <0 0 0>
        <5 0 0>
        <0 0 0>
        <0 0 0>
        <-5 0 0>
        <-8 0 0>
        <-4 0 0>
        <8 10 0>
        <8 0 0>
        <4 0 0>
        <-8 -10 0>
        <8 0 0>
        <0 0 0>
        <0 0 0>
        <5 0 0>
        <0 0 0>
        <-10 0 0>
        <-10 -22 0>
       ]}}
    )
    Skeletal Mesh=$User.SkeletalMesh
   )
   &ScaleColor(
    Scale Alpha=&FloatFromCurve(FloatCurve=DataInterface() {propertyValues {Curve {keys [
         {arriveTangent=23.8908, interpMode=RCIM_Cubic, leaveTangent=23.8908, tangentMode=RCTM_User, time=0, value=0}
         {arriveTangent=-2.3136, interpMode=RCIM_Cubic, leaveTangent=-2.3136, tangentMode=RCTM_User, time=0.1, value=1}
         <at 1 0 RCIM_Cubic>
        ]}}})
    ScaleA=true
    ScaleRGB=true
   )
   &DynamicMaterialParameters(
    Index 0 Param 1=&FloatFromCurve(FloatCurve=DataInterface() {propertyValues {Curve {keys [
         {interpMode=RCIM_Cubic, leaveTangentWeight=0.0435, tangentMode=RCTM_User, time=0, value=-0.2}
         {arriveTangent=18.2867, interpMode=RCIM_Cubic, leaveTangent=18.2867, tangentMode=RCTM_User, time=0.07, value=0.275}
         {arriveTangent=5.0806, interpMode=RCIM_Cubic, leaveTangent=5.0806, time=0.15, value=1}
         {arriveTangent=5.6764, interpMode=RCIM_Cubic, leaveTangent=5.6764, tangentMode=RCTM_User, time=1, value=5}
        ]}}})
    Index 0 Param 2=1
    Index 0 Param 3=&FloatFromCurve(FloatCurve=DataInterface() {propertyValues {Curve {keys [{arriveTangent=2.2883, interpMode=RCIM_Cubic, leaveTangent=2.2883, tangentMode=RCTM_User, time=0, value=-0.25}, <at 1 1.8 RCIM_Cubic>]}}})
    Module.Param0WriteEnabled=true
    Module.Param1WriteEnabled=true
    Module.Param2WriteEnabled=true
    Module.Param3WriteEnabled=false
   )
  ]
  rendererClasses [&NiagaraSpriteRendererProperties]
  renderers [{rendererClass=&NiagaraSpriteRendererProperties}]
 }
}
summary {
 systemName=NS_BrawnySmasher_SwapFX
 emitters [
  {emitterName=Flare, rendererClasses [&NiagaraSpriteRendererProperties, &NiagaraMeshRendererProperties]}
  {emitterName=Shield_Decals, rendererClasses [&NiagaraSpriteRendererProperties]}
  {emitterName=Ring, rendererClasses [&NiagaraMeshRendererProperties]}
  {emitterName=Fragments, rendererClasses [&NiagaraMeshRendererProperties]}
  {emitterName=Light, rendererClasses [&NiagaraLightRendererProperties]}
  {emitterName=Embers, simTarget=GPUComputeSim, rendererClasses [&NiagaraSpriteRendererProperties]}
 ]
 userVariables [
  User.BoneFlow {
   type=NiagaraDataInterfaceSkeletalMesh
   defaultValue {
    struct=NiagaraExt_VariableValue_DataInterface
    value {dataInterface=&self:NiagaraDataInterfaceSkeletalMesh_2}
   }
  }
  User.SkeletalMesh {
   type=NiagaraDataInterfaceSkeletalMesh
   defaultValue {
    struct=NiagaraExt_VariableValue_DataInterface
    value {dataInterface=&self:NiagaraDataInterfaceSkeletalMesh_0}
   }
  }
 ]
}
